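# Sound cue for deploying a jammer, apparently for a Season 17 quest, routed through the Fort_UI sound class.
# A Random node picks one of four equally weighted deploy waves and feeds it into a Mixer that boosts its level.
# A root Modulator then adds slight pitch and volume variation.
# A concurrency override allows only one instance at a time, stopping the oldest when a new one starts.
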SoundCue /Game/Athena/Sounds/Quests/Jammer/S17_Jammer_Deploy_Cue
root=Modulator_0
nodes [
 {id=Mixer_0, children [Random_1], properties {inputVolume [1.5]}}
 {id=Modulator_0, children [Mixer_0], properties {pitchMin=1, volumeMax=1.1, volumeMin=1}}
 {id=Random_1, children [WavePlayer_1,WavePlayer_5,WavePlayer_6,WavePlayer_7], properties {weights [1,1,1,1]}}
 {id=WavePlayer_1, properties {soundWaveAssetPtr=&/S17_Jammer_Deploy_01}}
 {id=WavePlayer_5, properties {soundWaveAssetPtr=&/S17_Jammer_Deploy_02}}
 {id=WavePlayer_6, properties {soundWaveAssetPtr=&/S17_Jammer_Deploy_03}}
 {id=WavePlayer_7, properties {soundWaveAssetPtr=&/S17_Jammer_Deploy_04}}
]
properties {
 bOverrideConcurrency=true
 maxDistance=2097152
 soundClassObject=&Fort_UI
 volumeMultiplier=1
 concurrencyOverrides {maxCount=1, resolutionRule=StopOldest}
}
