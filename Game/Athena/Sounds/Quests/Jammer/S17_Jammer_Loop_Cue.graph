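# Looping sound cue for the Season 17 Jammer quest object, playing its ambient loop in the world.
# A single looping wave player feeds one jammer loop wave directly to the output with a boosted volume.
# It routes to the quests sound class, uses a radar jammer attenuation asset for distance falloff, and primes on load.
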SoundCue /Game/Athena/Sounds/Quests/Jammer/S17_Jammer_Loop_Cue
root=WavePlayer_0
nodes [{id=WavePlayer_0, properties {bLooping=true, soundWaveAssetPtr=&/S17_Jammer_Loop_01}}]
properties {
 attenuationSettings=&/Fort_Radar_Jammer_Loop_Att
 bPrimeOnLoad=true
 maxDistance=2510
 soundClassObject=&Fort_Quests
 volumeMultiplier=1.5
 attenuationOverrides {falloffDistance=700, attenuationShapeExtents {x=100}}
}
